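# Idle fire effect for the Scavenger Fire backpack cosmetic, built from two sprite emitters.
# Embers_and_Sparks steadily spawns short-lived glowing sprites on an offset torus, pushed along its normal with random jitter, then shrinking, fading and scaling by speed.
# FlameDecal bursts one camera-offset flame sprite, placed on User.SkeletalMesh bones or sockets each update and made transparent when bisHandHidden is set.
# Most other user parameters, including speed and movement states, go unread here.
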NiagaraSystem /BRCosmetics/Accessories/FORT_Backpacks/M_MED_Dark_Viking_Fire/FX/NS_ScavengerFire_BackIdle
dependencies {
 usedDataInterfaces [&NiagaraDataInterfaceCurve, &NiagaraDataInterfaceColorCurve]
 usedDynamicInputs []
 usedModules [&SystemState]
}
emitters {
 Embers_and_Sparks {
  emitterUpdateScript [
   &EmitterState()
   &SpawnRate(SpawnRate=6, Use Spawn Probability=false)
  ]
  particleSpawnScript [
   &FN_SetSpawnGroup_v2(Number of Spawn Groups:NiagaraInt32=&MakeCustomIntFromBool(Bool=$User.bisHandHidden, True INT=0, False INT=1)) {enabled=false}
   &InitializeParticle(
    Color=<rgba 60 8.0591 0 1>
    Lifetime Max=0.6
    Lifetime Min=0.5
    Lifetime Mode=::NewEnumerator1
    Sprite Size Max=<2 3>
    Sprite Size Min=<1.5 2>
    Sprite Size Mode=::NewEnumerator2
    UsePositionOffset=false
    Write Material Random=true
   ) {module=InitializeParticle001}
   &TorusLocation(
    Align Torus With Axis=true
    Handle Radius=0.5
    Large Radius=11
    MaskSpawnGroup=false
    Module.Override Randomness=false
    Torus Axis=<0 1 0>
    Torus Origin:NiagaraPosition=/Niagara/DynamicInputs/Transforms/TransformPosition(Position=<0 -14 -5>, Source Space=::Local, Destination Space=::Simulation)
   )
   /SetVariables(Particles.EmbersSubImageIndex:NiagaraFloat=&UniformRangedFloat(Minimum=0, Maximum=36)) {isSetParameters=true}
   /SetVariables() {isSetParameters=true}
   /SetVariables(Particles.EmberSize:Vector2f=&Vector2DFromFloat(Value=&UniformRangedFloat(Minimum=1.5, Maximum=4))) {isSetParameters=true}
   &InheritVelocity(Module.Limit Speed=true, Velocity Scale=<0.1 0.1 0.1>)
   &AddVelocity(Velocity=&Add_Vector(A=$Output.TorusLocation.TorusNormal, B=&UniformRangedVector(Minimum=<-10 -10 -10>, Maximum=<10 10 30>)))
  ]
  particleUpdateScript [
   /SetVariables(Particles.EmberSize:Vector2f=&Multiply_Vector2(
     A=$Particles.Initial.EmberSize
     B=&Vector2DFromFloat(Value=&FloatFromCurve(FloatCurve=DataInterface() {propertyValues {Curve {keys [<at 0.5 1>,<at 1 0.5>]}}}))
    )) {isSetParameters=true}
   &Drag(Drag=2, Use Linear Drag=true, Use Rotational Drag=false) {enabled=false}
   &ParticleState()
   &SolveForcesAndVelocity(Clamp Velocity=false, Limit Acceleration=false)
   &ScaleColor(
    Scale Alpha=&FloatFromCurve(FloatCurve=DataInterface() {propertyValues {Curve {keys [<at 0 1>,<at 1 0>]}}})
    Scale RGB=&MakeVectorFromLinearColorRGB(Color=&ColorFromCurve(ColorCurve=DataInterface() {propertyValues {
        AlphaCurve {keys [<at 0 1 RCIM_Cubic>,<at 1 0 RCIM_Cubic>]}
        BlueCurve {keys [<at 0 1 RCIM_Cubic>,<at 1 0 RCIM_Cubic>]}
        GreenCurve {keys [<at 0 1 RCIM_Cubic>,<at 1 0 RCIM_Cubic>]}
        RedCurve {keys [<at 0 1 RCIM_Cubic>,<at 1 0 RCIM_Cubic>]}
       }}))
    ScaleA=true
    ScaleRGB=true
   )
   &ScaleSpriteSizeBySpeed(Max Scale Factor=<1 2>, Min Scale Factor=<0.5 1>)
  ]
  rendererClasses [&NiagaraSpriteRendererProperties]
  renderers [
   {rendererClass=&NiagaraSpriteRendererProperties}
   {rendererClass=&NiagaraSpriteRendererProperties, rendererIndex=1}
   {rendererClass=&NiagaraSpriteRendererProperties, rendererIndex=2}
  ]
 }
 FlameDecal {
  emitterUpdateScript [
   &EmitterState(Inactive Response=::NewEnumerator2, Life Cycle Mode=::NewEnumerator1, Loop Duration=0.4, UseLoopDelay=false)
   &SpawnBurst_Instantaneous(Spawn Count=1, Use Spawn Probability=false)
  ]
  particleSpawnScript [
   /Niagara/Modules/Spawn/Initialization/InitializeParticle(
    Color=<rgba 6 1.083 0 0.75>
    Mesh Scale=<1.25 0.75 1>
    Module.Write Color=true
    Module.Write Lifetime=true
    Module.Write Mass=false
    Module.Write Material Random=true
    Module.Write Position=false
    Module.Write Scale=true
    Module.Write SpriteRotation=false
    Module.Write SpriteSize=true
    Sprite Size=&RandomRangeVector2D(Minimum=<0 0>, Maximum=<1 1>)
   ) {enabled=false}
   &CameraOffset(Camera Offset Amount=5)
   /Niagara/Modules/Spawn/Initialization/V2/InitializeParticle(
    Color=<rgba 12 2.166 0 0.75>
    Sprite Size Mode=::NewEnumerator1
    Uniform Sprite Size Max=12
    Uniform Sprite Size Min=6
    UsePositionOffset=false
    Write Material Random=true
   ) {module=InitializeParticle001}
  ]
  particleUpdateScript [
   &ParticleState()
   /Niagara/Modules/Spawn/Location/SphereLocation(Hemisphere X=true, Hemisphere Y=true, Hemisphere Z=true, MaskSpawnGroup=false, Module.Override Randomness=false, Surface Only=false) {enabled=false}
   &SkeletalMeshLocation(
    Bone / Socket Index=&RandomRangeInt(Minimum=2, Maximum=3)
    Inherit Velocity=false
    MaskSpawnGroup=false
    Mesh Sampling Type=::NewEnumerator5
    Sampled Position Offset Coordinate Space:ENiagaraMeshTransforms=::NewEnumerator2
    Skeletal Mesh=$User.SkeletalMesh
    Socket Sampling Mode=::NewEnumerator5
   )
   &ScaleColor(
    Scale Alpha=&MakeCustomFloatFromBool(Bool=$User.bisHandHidden, True Float=0, False Float=1)
    ScaleA=true
    ScaleRGB=true
   )
  ]
  rendererClasses [&NiagaraSpriteRendererProperties]
  renderers [
   {rendererClass=&NiagaraSpriteRendererProperties}
   {rendererClass=&NiagaraSpriteRendererProperties, rendererIndex=1}
  ]
 }
}
summary {
 systemName=NS_ScavengerFire_BackIdle
 emitters [
  {emitterName=Embers_and_Sparks, rendererClasses [&NiagaraSpriteRendererProperties]}
  {emitterName=FlameDecal, rendererClasses [&NiagaraSpriteRendererProperties]}
 ]
 userVariables [
  User.bIsFrontEndPreview {type=NiagaraBool}
  User.bIsFullyDeployed {type=NiagaraBool}
  User.bisHandHidden {type=NiagaraBool}
  User.FlameSpawnRate {type=NiagaraFloat, defaultValue=0, description=""}
  User.IsSkydiving {type=NiagaraBool}
  User.IsSprinting {type=NiagaraBool}
  User.PlayerSpeed {type=NiagaraFloat, defaultValue=0, description=""}
  User.RibbonIDOffset {type=NiagaraInt32}
  User.SkeletalMesh {
   type=NiagaraDataInterfaceSkeletalMesh
   defaultValue {
    struct=NiagaraExt_VariableValue_DataInterface
    value {dataInterface=&self:NiagaraDataInterfaceSkeletalMesh_0}
   }
  }
 ]
}
