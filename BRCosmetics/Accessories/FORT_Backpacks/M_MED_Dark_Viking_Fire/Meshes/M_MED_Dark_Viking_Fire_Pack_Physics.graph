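# Physics asset for the winged Dark Viking Fire backpack cosmetic, defining the simulation bodies and joints for its left and right wing bones.
# Each wing has a large root capsule with six chains, a through f, of smaller capsules, plus c2 and d2 branches off the c and d bases.
# Angular limits are tight throughout, stiffest at the root joints and slightly looser toward the tips of the longer chains.
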PhysicsAsset /BRCosmetics/Accessories/FORT_Backpacks/M_MED_Dark_Viking_Fire/Meshes/M_MED_Dark_Viking_Fire_Pack_Physics
bodies {
 dyn_wing_l_01 [{shapeType=Capsule, length=16.9979, radius=9.9477}]
 dyn_wing_l_a_01 [{shapeType=Capsule, length=15.3479, radius=2.4821, center=<11.8785 -0.0023 -0.1852>, rotation=<rot 0 89.1072 -89.7911>}]
 dyn_wing_l_a_02 [{shapeType=Capsule, length=11.8956, radius=2, center=<8.3154 0 0>, rotation=<rot 90 0 0>}]
 dyn_wing_l_a_03 [{shapeType=Capsule, length=11.8956, radius=2, center=<8.3154 0 0>, rotation=<rot 90 0 0>}]
 dyn_wing_l_a_04 [{shapeType=Capsule, length=11.8956, radius=2, center=<8.3154 0 0>, rotation=<rot 90 0 0>}]
 dyn_wing_l_b_01 [{shapeType=Capsule, length=14.3606, radius=3.0515, center=<9.5129 0.949 -0.8501>, rotation=<rot 0 88.2789 -89.3658>}]
 dyn_wing_l_b_02 [{shapeType=Capsule, length=11.8956, radius=2, center=<8.3154 0 0>, rotation=<rot 90 0 0>}]
 dyn_wing_l_b_03 [{shapeType=Capsule, length=11.8956, radius=2, center=<8.3154 0 0>, rotation=<rot 90 0 0>}]
 dyn_wing_l_b_04 [{shapeType=Capsule, length=11.8956, radius=2, center=<8.3154 0 0>, rotation=<rot 90 0 0>}]
 dyn_wing_l_c2_01 [{shapeType=Capsule, length=11.8956, radius=2, center=<8.3154 0 0>, rotation=<rot 90 0 0>}]
 dyn_wing_l_c2_02 [{shapeType=Capsule, length=11.8956, radius=2, center=<8.3154 0 0>, rotation=<rot 90 0 0>}]
 dyn_wing_l_c2_03 [{shapeType=Capsule, length=11.8956, radius=2, center=<8.3154 0 0>, rotation=<rot 90 0 0>}]
 dyn_wing_l_c_01 [{shapeType=Capsule, length=12.0996, radius=3.3793, center=<10.8647 0.0425 -0.3344>, rotation=<rot 0 85.51 -87.0798>}]
 dyn_wing_l_c_02 [{shapeType=Capsule, length=11.8956, radius=2, center=<8.3154 0 0>, rotation=<rot 90 0 0>}]
 dyn_wing_l_c_03 [{shapeType=Capsule, length=11.8956, radius=2, center=<8.3154 0 0>, rotation=<rot 90 0 0>}]
 dyn_wing_l_c_04 [{shapeType=Capsule, length=11.8956, radius=2, center=<8.3154 0 0>, rotation=<rot 90 0 0>}]
 dyn_wing_l_d2_01 [{shapeType=Capsule, length=11.8956, radius=2, center=<8.3154 0 0>, rotation=<rot 90 0 0>}]
 dyn_wing_l_d2_02 [{shapeType=Capsule, length=11.8956, radius=2, center=<8.3154 0 0>, rotation=<rot 90 0 0>}]
 dyn_wing_l_d_01 [{shapeType=Capsule, length=13.6416, radius=3.6791, center=<7.865 1.3965 0.4275>, rotation=<rot 0 88.4448 -89.038>}]
 dyn_wing_l_d_02 [{shapeType=Capsule, length=11.8956, radius=2, center=<8.3154 0 0>, rotation=<rot 90 0 0>}]
 dyn_wing_l_d_03 [{shapeType=Capsule, length=11.8956, radius=2, center=<8.3154 0 0>, rotation=<rot 90 0 0>}]
 dyn_wing_l_d_04 [{shapeType=Capsule, length=11.8956, radius=2, center=<8.3154 0 0>, rotation=<rot 90 0 0>}]
 dyn_wing_l_e_01 [{shapeType=Capsule, length=10.8144, radius=2.4611, center=<5.5416 -0.5674 -0.327>, rotation=<rot 0 84.4721 -89.7323>}]
 dyn_wing_l_e_02 [{shapeType=Capsule, length=11.8956, radius=2, center=<8.3154 0 0>, rotation=<rot 90 0 0>}]
 dyn_wing_l_e_03 [{shapeType=Capsule, length=11.8956, radius=2, center=<8.3154 0 0>, rotation=<rot 90 0 0>}]
 dyn_wing_l_f_01 [{shapeType=Capsule, length=4.9415, radius=2, center=<4.4888 0 0.0002>, rotation=<rot 90 0 0>}]
 dyn_wing_l_f_02 [{shapeType=Capsule, length=11.8956, radius=2, center=<8.3154 0 0>, rotation=<rot 90 0 0>}]
 dyn_wing_l_f_03 [{shapeType=Capsule, length=11.8956, radius=2, center=<8.3154 0 0>, rotation=<rot 90 0 0>}]
 dyn_wing_r_01 [{shapeType=Capsule, length=17.2323, radius=11.5229}]
 dyn_wing_r_a_01 [{shapeType=Capsule, length=15.3477, radius=2.4821, center=<11.8781 0.002 -0.1831>, rotation=<rot 0 90.8947 -89.7393>}]
 dyn_wing_r_a_02 [{shapeType=Capsule, length=12.919, radius=2, center=<8.2859 0 -0.0002>, rotation=<rot 90 0 0>}]
 dyn_wing_r_a_03 [{shapeType=Capsule, length=12.919, radius=2, center=<8.2859 0 -0.0002>, rotation=<rot 90 0 0>}]
 dyn_wing_r_a_04 [{shapeType=Capsule, length=12.919, radius=2, center=<8.2859 0 -0.0002>, rotation=<rot 90 0 0>}]
 dyn_wing_r_b_01 [{shapeType=Capsule, length=14.375, radius=3.1318, center=<9.5061 -0.8716 -0.879>, rotation=<rot 0 92.3881 -88.923>}]
 dyn_wing_r_b_02 [{shapeType=Capsule, length=12.919, radius=2, center=<8.2859 0 -0.0002>, rotation=<rot 90 0 0>}]
 dyn_wing_r_b_03 [{shapeType=Capsule, length=12.919, radius=2, center=<8.2859 0 -0.0002>, rotation=<rot 90 0 0>}]
 dyn_wing_r_b_04 [{shapeType=Capsule, length=12.919, radius=2, center=<8.2859 0 -0.0002>, rotation=<rot 90 0 0>}]
 dyn_wing_r_c2_01 [{shapeType=Capsule, length=12.919, radius=2, center=<8.2859 0 -0.0002>, rotation=<rot 90 0 0>}]
 dyn_wing_r_c2_02 [{shapeType=Capsule, length=12.919, radius=2, center=<8.2859 0 -0.0002>, rotation=<rot 90 0 0>}]
 dyn_wing_r_c2_03 [{shapeType=Capsule, length=12.919, radius=2, center=<8.2859 0 -0.0002>, rotation=<rot 90 0 0>}]
 dyn_wing_r_c_01 [{shapeType=Capsule, length=12.0992, radius=3.3837, center=<10.8649 -0.0403 -0.3301>, rotation=<rot 0 94.5226 -87.1194>}]
 dyn_wing_r_c_02 [{shapeType=Capsule, length=12.919, radius=2, center=<8.2859 0 -0.0002>, rotation=<rot 90 0 0>}]
 dyn_wing_r_c_03 [{shapeType=Capsule, length=12.919, radius=2, center=<8.2859 0 -0.0002>, rotation=<rot 90 0 0>}]
 dyn_wing_r_c_04 [{shapeType=Capsule, length=12.919, radius=2, center=<8.2859 0 -0.0002>, rotation=<rot 90 0 0>}]
 dyn_wing_r_d2_01 [{shapeType=Capsule, length=12.919, radius=2, center=<8.2859 0 -0.0002>, rotation=<rot 90 0 0>}]
 dyn_wing_r_d2_02 [{shapeType=Capsule, length=12.919, radius=2, center=<8.2859 0 -0.0002>, rotation=<rot 90 0 0>}]
 dyn_wing_r_d_01 [{shapeType=Capsule, length=13.6542, radius=3.7105, center=<7.8637 -1.3628 0.4216>, rotation=<rot 0 91.8975 -88.9348>}]
 dyn_wing_r_d_02 [{shapeType=Capsule, length=12.919, radius=2, center=<8.2859 0 -0.0002>, rotation=<rot 90 0 0>}]
 dyn_wing_r_d_03 [{shapeType=Capsule, length=12.919, radius=2, center=<8.2859 0 -0.0002>, rotation=<rot 90 0 0>}]
 dyn_wing_r_d_04 [{shapeType=Capsule, length=12.919, radius=2, center=<8.2859 0 -0.0002>, rotation=<rot 90 0 0>}]
 dyn_wing_r_e_01 [{shapeType=Capsule, length=10.8121, radius=2.4581, center=<5.5388 0.5749 -0.321>, rotation=<rot 0 95.322 -89.6341>}]
 dyn_wing_r_e_02 [{shapeType=Capsule, length=12.919, radius=2, center=<8.2859 0 -0.0002>, rotation=<rot 90 0 0>}]
 dyn_wing_r_e_03 [{shapeType=Capsule, length=12.919, radius=2, center=<8.2859 0 -0.0002>, rotation=<rot 90 0 0>}]
 dyn_wing_r_f_01 [{shapeType=Capsule, length=4.9415, radius=2, center=<4.4888 0 0.0002>, rotation=<rot 90 0 0>}]
 dyn_wing_r_f_02 [{shapeType=Capsule, length=12.919, radius=2, center=<8.2859 0 -0.0002>, rotation=<rot 90 0 0>}]
 dyn_wing_r_f_03 [{shapeType=Capsule, length=12.919, radius=2, center=<8.2859 0 -0.0002>, rotation=<rot 90 0 0>}]
}
constraints [
 {bone1Name=dyn_wing_r_f_02, bone2Name=dyn_wing_r_f_01, swing1LimitDegrees=1, swing2LimitDegrees=1, twistLimitDegrees=1}
 {bone1Name=dyn_wing_l_a_03, bone2Name=dyn_wing_l_a_02, swing1LimitDegrees=5, swing2LimitDegrees=5, twistLimitDegrees=2}
 {bone1Name=dyn_wing_l_a_04, bone2Name=dyn_wing_l_a_03, swing1LimitDegrees=5, swing2LimitDegrees=5, twistLimitDegrees=2}
 {bone1Name=dyn_wing_l_a_02, bone2Name=dyn_wing_l_a_01, swing1LimitDegrees=2, swing2LimitDegrees=2, twistLimitDegrees=2}
 {bone1Name=dyn_wing_l_b_04, bone2Name=dyn_wing_l_b_03, swing1LimitDegrees=5, swing2LimitDegrees=5, twistLimitDegrees=2}
 {bone1Name=dyn_wing_l_b_03, bone2Name=dyn_wing_l_b_02, swing1LimitDegrees=5, swing2LimitDegrees=5, twistLimitDegrees=2}
 {bone1Name=dyn_wing_l_b_02, bone2Name=dyn_wing_l_b_01, swing1LimitDegrees=2, swing2LimitDegrees=2, twistLimitDegrees=2}
 {bone1Name=dyn_wing_l_c2_03, bone2Name=dyn_wing_l_c2_02, swing1LimitDegrees=5, swing2LimitDegrees=5, twistLimitDegrees=2}
 {bone1Name=dyn_wing_l_c2_02, bone2Name=dyn_wing_l_c2_01, swing1LimitDegrees=5, swing2LimitDegrees=5, twistLimitDegrees=2}
 {bone1Name=dyn_wing_l_c2_01, bone2Name=dyn_wing_l_c_01, swing1LimitDegrees=2, swing2LimitDegrees=2, twistLimitDegrees=2}
 {bone1Name=dyn_wing_l_c_04, bone2Name=dyn_wing_l_c_03, swing1LimitDegrees=5, swing2LimitDegrees=5, twistLimitDegrees=2}
 {bone1Name=dyn_wing_l_c_03, bone2Name=dyn_wing_l_c_02, swing1LimitDegrees=5, swing2LimitDegrees=5, twistLimitDegrees=2}
 {bone1Name=dyn_wing_l_c_02, bone2Name=dyn_wing_l_c_01, swing1LimitDegrees=2, swing2LimitDegrees=2, twistLimitDegrees=2}
 {bone1Name=dyn_wing_l_d_04, bone2Name=dyn_wing_l_d_03, swing1LimitDegrees=5, swing2LimitDegrees=2, twistLimitDegrees=2}
 {bone1Name=dyn_wing_l_d_03, bone2Name=dyn_wing_l_d_02, swing1LimitDegrees=5, swing2LimitDegrees=2, twistLimitDegrees=2}
 {bone1Name=dyn_wing_l_d_02, bone2Name=dyn_wing_l_d_01, swing1LimitDegrees=2, swing2LimitDegrees=2, twistLimitDegrees=2}
 {bone1Name=dyn_wing_l_d2_02, bone2Name=dyn_wing_l_d2_01, swing1LimitDegrees=2, swing2LimitDegrees=2, twistLimitDegrees=2}
 {bone1Name=dyn_wing_l_d2_01, bone2Name=dyn_wing_l_d_01, swing1LimitDegrees=2, swing2LimitDegrees=2, twistLimitDegrees=2}
 {bone1Name=dyn_wing_l_e_03, bone2Name=dyn_wing_l_e_02, swing1LimitDegrees=2, swing2LimitDegrees=2, twistLimitDegrees=2}
 {bone1Name=dyn_wing_l_e_02, bone2Name=dyn_wing_l_e_01, swing1LimitDegrees=2, swing2LimitDegrees=2, twistLimitDegrees=2}
 {bone1Name=dyn_wing_l_f_03, bone2Name=dyn_wing_l_f_02, swing1LimitDegrees=2, swing2LimitDegrees=2, twistLimitDegrees=2}
 {bone1Name=dyn_wing_l_f_02, bone2Name=dyn_wing_l_f_01, swing1LimitDegrees=1, swing2LimitDegrees=1, twistLimitDegrees=1}
 {bone1Name=dyn_wing_r_e_02, bone2Name=dyn_wing_r_e_01, swing1LimitDegrees=2, swing2LimitDegrees=2, twistLimitDegrees=2}
 {bone1Name=dyn_wing_r_e_03, bone2Name=dyn_wing_r_e_02, swing1LimitDegrees=2, swing2LimitDegrees=2, twistLimitDegrees=2}
 {bone1Name=dyn_wing_r_f_03, bone2Name=dyn_wing_r_f_02, swing1LimitDegrees=2, swing2LimitDegrees=2, twistLimitDegrees=2}
 {bone1Name=dyn_wing_r_d2_01, bone2Name=dyn_wing_r_d_01, swing1LimitDegrees=2, swing2LimitDegrees=2, twistLimitDegrees=2}
 {bone1Name=dyn_wing_r_d2_02, bone2Name=dyn_wing_r_d2_01, swing1LimitDegrees=2, swing2LimitDegrees=2, twistLimitDegrees=2}
 {bone1Name=dyn_wing_r_d_02, bone2Name=dyn_wing_r_d_01, swing1LimitDegrees=2, swing2LimitDegrees=2, twistLimitDegrees=2}
 {bone1Name=dyn_wing_r_d_03, bone2Name=dyn_wing_r_d_02, swing1LimitDegrees=5, swing2LimitDegrees=2, twistLimitDegrees=2}
 {bone1Name=dyn_wing_r_d_04, bone2Name=dyn_wing_r_d_03, swing1LimitDegrees=5, swing2LimitDegrees=2, twistLimitDegrees=2}
 {bone1Name=dyn_wing_r_c_04, bone2Name=dyn_wing_r_c_03, swing1LimitDegrees=5, swing2LimitDegrees=5, twistLimitDegrees=2}
 {bone1Name=dyn_wing_r_c_03, bone2Name=dyn_wing_r_c_02, swing1LimitDegrees=5, swing2LimitDegrees=5, twistLimitDegrees=2}
 {bone1Name=dyn_wing_r_c_02, bone2Name=dyn_wing_r_c_01, swing1LimitDegrees=2, swing2LimitDegrees=2, twistLimitDegrees=2}
 {bone1Name=dyn_wing_r_c2_03, bone2Name=dyn_wing_r_c2_02, swing1LimitDegrees=5, swing2LimitDegrees=5, twistLimitDegrees=2}
 {bone1Name=dyn_wing_r_c2_02, bone2Name=dyn_wing_r_c2_01, swing1LimitDegrees=5, swing2LimitDegrees=5, twistLimitDegrees=2}
 {bone1Name=dyn_wing_r_c2_01, bone2Name=dyn_wing_r_c_01, swing1LimitDegrees=2, swing2LimitDegrees=2, twistLimitDegrees=2}
 {bone1Name=dyn_wing_r_b_02, bone2Name=dyn_wing_r_b_01, swing1LimitDegrees=2, swing2LimitDegrees=2, twistLimitDegrees=2}
 {bone1Name=dyn_wing_r_b_03, bone2Name=dyn_wing_r_b_02, swing1LimitDegrees=5, swing2LimitDegrees=5, twistLimitDegrees=2}
 {bone1Name=dyn_wing_r_b_04, bone2Name=dyn_wing_r_b_03, swing1LimitDegrees=5, swing2LimitDegrees=5, twistLimitDegrees=2}
 {bone1Name=dyn_wing_r_a_02, bone2Name=dyn_wing_r_a_01, swing1LimitDegrees=2, swing2LimitDegrees=2, twistLimitDegrees=2}
 {bone1Name=dyn_wing_r_a_03, bone2Name=dyn_wing_r_a_02, swing1LimitDegrees=5, swing2LimitDegrees=5, twistLimitDegrees=2}
 {bone1Name=dyn_wing_r_a_04, bone2Name=dyn_wing_r_a_03, swing1LimitDegrees=5, swing2LimitDegrees=5, twistLimitDegrees=2}
 {bone1Name=dyn_wing_r_e_01, bone2Name=dyn_wing_r_01, swing1LimitDegrees=1, swing2LimitDegrees=1, twistLimitDegrees=1}
 {bone1Name=dyn_wing_r_f_01, bone2Name=dyn_wing_r_01, swing1LimitDegrees=1, swing2LimitDegrees=1, twistLimitDegrees=1}
 {bone1Name=dyn_wing_r_d_01, bone2Name=dyn_wing_r_01, swing1LimitDegrees=1, swing2LimitDegrees=1, twistLimitDegrees=1}
 {bone1Name=dyn_wing_r_c_01, bone2Name=dyn_wing_r_01, swing1LimitDegrees=1, swing2LimitDegrees=1, twistLimitDegrees=1}
 {bone1Name=dyn_wing_r_b_01, bone2Name=dyn_wing_r_01, swing1LimitDegrees=1, swing2LimitDegrees=1, twistLimitDegrees=1}
 {bone1Name=dyn_wing_r_a_01, bone2Name=dyn_wing_r_01, swing1LimitDegrees=1, swing2LimitDegrees=1, twistLimitDegrees=1}
 {bone1Name=dyn_wing_l_f_01, bone2Name=dyn_wing_l_01, swing1LimitDegrees=1, swing2LimitDegrees=1, twistLimitDegrees=1}
 {bone1Name=dyn_wing_l_e_01, bone2Name=dyn_wing_l_01, swing1LimitDegrees=1, swing2LimitDegrees=1, twistLimitDegrees=1}
 {bone1Name=dyn_wing_l_d_01, bone2Name=dyn_wing_l_01, swing1LimitDegrees=1, swing2LimitDegrees=1, twistLimitDegrees=1}
 {bone1Name=dyn_wing_l_c_01, bone2Name=dyn_wing_l_01, swing1LimitDegrees=1, swing2LimitDegrees=1, twistLimitDegrees=1}
 {bone1Name=dyn_wing_l_b_01, bone2Name=dyn_wing_l_01, swing1LimitDegrees=1, swing2LimitDegrees=1, twistLimitDegrees=1}
 {bone1Name=dyn_wing_l_a_01, bone2Name=dyn_wing_l_01, swing1LimitDegrees=1, swing2LimitDegrees=1, twistLimitDegrees=1}
]
