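# Physics asset for the top wing mesh of the Dark Viking Fire backpack cosmetic, defining bodies and joint limits for its dynamic wing bones.
# Each side has a separate three-bone chain of small capsules running from two wing base bones to a wing bone, with no constraint linking left and right.
# Narrow angular limits keep movement subtle, with the base joints allowing less swing than the outer wing joints.
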PhysicsAsset /BRCosmetics/Accessories/FORT_Backpacks/M_MED_Dark_Viking_Fire/Meshes/M_MED_Dark_Viking_Fire_Pack_Wing_Top_Physics
bodies {
 dyn_wing_base_l_02 [{shapeType=Capsule, length=2, radius=2, rotation=<rot 90 0 0>}]
 dyn_wing_base_l_03 [{shapeType=Capsule, length=2, radius=2, center=<0 -2.7167 0.0001>, rotation=<rot 0 0 90>}]
 dyn_wing_base_r_02 [{shapeType=Capsule, length=2, radius=2, rotation=<rot 90 0 0>}]
 dyn_wing_base_r_03 [{shapeType=Capsule, length=2, radius=2, center=<2.8207 0 0.0001>, rotation=<rot 90 0 0>}]
 dyn_wing_l_01 [{shapeType=Capsule, length=2, radius=2, rotation=<rot 90 0 0>}]
 dyn_wing_r_01 [{shapeType=Capsule, length=2, radius=2, rotation=<rot 90 0 0>}]
}
constraints [
 {bone1Name=dyn_wing_l_01, bone2Name=dyn_wing_base_l_03, swing1LimitDegrees=10, swing2LimitDegrees=5, twistLimitDegrees=8}
 {bone1Name=dyn_wing_r_01, bone2Name=dyn_wing_base_r_03, swing1LimitDegrees=10, swing2LimitDegrees=5, twistLimitDegrees=8}
 {bone1Name=dyn_wing_base_r_03, bone2Name=dyn_wing_base_r_02, swing1LimitDegrees=2, swing2LimitDegrees=2, twistLimitDegrees=8}
 {bone1Name=dyn_wing_base_l_03, bone2Name=dyn_wing_base_l_02, swing1LimitDegrees=2, swing2LimitDegrees=2, twistLimitDegrees=8}
]
